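; AxisThrow boss NPC pawn that starts seated on a throne and can join a player's team when downed.
; On spawn it finds the nearest throne, ignores its collision and moves onto it; proximity spheres drive leaning and wake-up events, and throne damage can also trigger the attack.
; When DBNO it drops team-size loot, turns invincible, then after a timer plays a montage and becomes convertible, optionally auto-converting to its last player attacker.
; Minimap marker and ambient whispers follow its state.
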(event Custom|BindToConvertComponent
  (ConvertEvents|BindEventtoOnConvertedEvent (Variables|FortPlayerPawn|GetConvertComponent) (EventDispatchers|CreateEvent :!event Custom|OnConverted)))

(event EventBeginPlay
  ;[1 1 1] BeginPlay
  (|CallParentFunction)
  (EventDispatchers|AssignOnDamaged (delegate Custom|OnDamaged_Event))
  (EventDispatchers|BindEventtoOnPlayerStartDBNO (EventDispatchers|CreateEvent :!event Custom|BossDBNO))
  (CallFunction|SetNPCAnimSpeedScalar 0.75)
  (CallFunction|TryDisableCollisionswithThrone)
  (Networking|SwitchHasAuthority
    (:Authority
      (CallFunction|BindtoConvertComponent)
      (Utilities|FlowControl|Delay 1.0)
      (Ability|Async|WaitGameplayEventToActor 0 (Variables|Default|GetEventTag_IntroDone) true
        (:then
          (bind (_gamestate _playlist) (Class|FortAsyncActionCurrentPlaylistReady|PlaylistDataReadyAsync)
            (:OnReady
              (Variables|Default|SetPlaylistTeamSize (Variables|/Script/FortniteGame.FortPlaylist|GetMaxTeamSize _playlist))
              (Components|Sphere|SetSphereRadius (Variables|Default|GetProximityTrigger) (ScalableFloat|GetValueatLevel (Variables|Default|GetProximitySphereRadius)) false)
              (Components|Sphere|SetSphereRadius (Variables|Default|GetAttackProximityTrigger) (ScalableFloat|GetValueatLevel (Variables|Default|GetAttackSphereRadius)) false)
              ;[1 1 1] Boss Map Marker
              (if (ScalableFloat|GetValueasBool (Variables|Default|GetShowBossMapMarker))
                (Components|Activation|Activate (Variables|Default|GetAthenaSpecialActor) false)))))
        (:EventReceived
          (Variables|Default|SetCanCollideWithThrone true))))
    (:Remote
      (Components|DestroyComponent (Variables|Default|GetProximityTrigger))
      (Variables|Default|GetbIsConverted
        (:then)
        (:else
          (Components|Activation|Activate (Variables|Default|GetVO%20Amb%20Whispers%20Component) false))))))

(event Custom|OnDamaged_Event
  (params (DamagedActor :type Actor%20Object%20Reference) (Damage :type Float%20%28single-precision%29) (InstigatedBy :type Controller%20Object%20Reference) (DamageCauser :type Actor%20Object%20Reference) (HitLocation :type Vector) (FHitComponent :type Primitive%20Component%20Object%20Reference) (BoneName :type Name) (Momentum :type Vector))
  ;[1 0.9544 0] Keep last damaging pawn to know team for conversion
  (bind _asfort_player_controller (Utilities|Casting|CastToFortPlayerController InstigatedBy))
  (CallFunction|AssignLastDamagingPawn (Game|Player|GetPlayerPawn _asfort_player_controller)))

(event Custom|TryDisableCollisionsWithThrone
  ;[1 1 1] Find the Throne and sit on it. In PIE the AI can spawn before the Throne is in the map.
  (if (not (Variables|Default|GetCanCollideWithThrone))
    (bind (_foundthrone _throne) (CallFunction|GetClosestThrone))
    (if _foundthrone
      (Utilities|IsValid _throne
        (:Is%20Valid
          (Variables|Default|SetThroneActor _throne)
          (CallFunction|DisableThroneCollisions)))
      (else
        (Utilities|FlowControl|Delay 1.0)
        (CallFunction|TryDisableCollisionswithThrone)))))

(event Custom|DisableThroneCollisions
  (bind _returnvalue (Ability|GetAbilitySystemComponent 0))
  (bind _throneactor (Variables|Default|GetThroneActor))
  ;[1 1 1] Disable Collisions With Trhone
  (Collision|IgnoreActorwhenMoving (Variables|Character|GetCapsuleComponent) _throneactor true)
  (Collision|IgnoreActorwhenMoving (Variables|Character|GetMesh) _throneactor true)
  (Networking|SwitchHasAuthority
    (:Authority
      (Utilities|IsValid _returnvalue
        (:Is%20Valid
          (Utilities|FlowControl|Delay 0.1)
          (Abilities|TryActivateAbilitiesbyTag _returnvalue (Variables|Default|GetTag_PushBack) true)
          (Utilities|FlowControl|Delay 0.1))
        (:Is%20Not%20Valid
          _))
      (Transformation|SetActorLocation self (+ (Math|Vector|MakeVector (.x (Transformation|GetActorLocation _throneactor)) (.y (Transformation|GetActorLocation _throneactor)) (+ (.z (Transformation|GetActorLocation _throneactor)) (+ (Components|Capsule|GetScaledCapsuleHalfHeight (Variables|Character|GetCapsuleComponent)) 15.0))) (* (Transformation|GetActorRightVector _throneactor) 5e+01)) false true))))

(event OnComponentBeginOverlap(ProximityTrigger)
  (params (OverlappedComponent :type Primitive%20Component%20Object%20Reference) (OtherActor :type Actor%20Object%20Reference) (OtherComp :type Primitive%20Component%20Object%20Reference) (OtherBodyIndex :type Integer) (bFromSweep :type Boolean) (SweepResult :type Hit%20Result%20Structure%20%28by%20ref%29))
  (bind _self self)
  ;[1 1 1] Track players getting close to make Boss Lean In and Out
  (Networking|SwitchHasAuthority
    (:Authority
      (bind _asfort_player_pawn (Utilities|Casting|CastToFortPlayerPawn OtherActor))
      (if (and (!= _asfort_player_pawn _self) (not (GameplayTags|HasMatchingGameplayTag _asfort_player_pawn "(TagName=\"Pawn.Athena.NonParticipant\")")))
        (Math|Integer|IncrementInt (Variables|Default|GetPlayersInProximityCount)
          (:%20%20
            (if (not (Variables|Default|GetBossIsLeaning))
              (Ability|SendGameplayEventtoActor _self (Variables|Default|GetTag_EventWhenPawnInProximity))
              (Variables|Default|SetBossIsLeaning true))))))))

(event Custom|ReEnableCollisionsWithThrone
  (bind _throneactor (Variables|Default|GetThroneActor))
  ;[1 1 1] Re-Enable Collisions with Throne
  (Collision|IgnoreActorwhenMoving (Variables|Character|GetCapsuleComponent) _throneactor false)
  (Collision|IgnoreActorwhenMoving (Variables|Character|GetMesh) _throneactor false))

(event OnComponentEndOverlap(ProximityTrigger)
  (params (OverlappedComponent :type Primitive%20Component%20Object%20Reference) (OtherActor :type Actor%20Object%20Reference) (OtherComp :type Primitive%20Component%20Object%20Reference) (OtherBodyIndex :type Integer))
  ;[1 1 1] Track players getting close to make Boss Lean In and Out
  (Networking|SwitchHasAuthority
    (:Authority
      (bind _asfort_player_pawn (Utilities|Casting|CastToFortPlayerPawn OtherActor))
      (if (and (!= _asfort_player_pawn self) (not (GameplayTags|HasMatchingGameplayTag _asfort_player_pawn "(TagName=\"Pawn.Athena.NonParticipant\")")))
        (bind _result (Math|Integer|DecrementInt (Variables|Default|GetPlayersInProximityCount))
          (:%20%20
            (if (<= _result 0)
              (Ability|SendGameplayEventtoActor self (Variables|Default|GetTag_EventWhenNoPawnsInProximity))
              (Variables|Default|SetBossIsLeaning false))))))))

(event EventOnDied_Event
  (params (DamagedActor :type Actor%20Object%20Reference) (Damage :type Float%20%28single-precision%29) (InstigatedBy :type Controller%20Object%20Reference) (DamageCauser :type Actor%20Object%20Reference) (HitLocation :type Vector :default "0, 0, 0") (FHitComponent :type Primitive%20Component%20Object%20Reference) (BoneName :type Name :default "None") (Momentum :type Vector :default "0, 0, 0"))
  (|CallParentFunction DamagedActor Damage InstigatedBy DamageCauser HitLocation FHitComponent BoneName Momentum)
  ;[1 1 1] OnDied - disable the minimap UI immediately instead of waiting to be cleaned up FORT-769042
  (if (> (ScalableFloat|GetValueatLevel (Variables|Default|GetRemoveMinimapUIOnDeath)) 0)
    (Components|Activation|Deactivate (Variables|Default|GetBossD_SpecialActor))))

(event OnComponentBeginOverlap(AttackProximityTrigger)
  (params (OverlappedComponent :type Primitive%20Component%20Object%20Reference) (OtherActor :type Actor%20Object%20Reference) (OtherComp :type Primitive%20Component%20Object%20Reference) (OtherBodyIndex :type Integer) (bFromSweep :type Boolean) (SweepResult :type Hit%20Result%20Structure%20%28by%20ref%29))
  ;[1 1 1] Players got WAY too close, attack!
  (Networking|SwitchHasAuthority
    (:Authority
      (bind _asfort_player_pawn (Utilities|Casting|CastToFortPlayerPawn OtherActor))
      (if (and (!= _asfort_player_pawn self) (not (GameplayTags|HasMatchingGameplayTag _asfort_player_pawn "(TagName=\"Pawn.Athena.NonParticipant\")")))
        (CallFunction|WakeUpAIFromIntro _asfort_player_pawn)))))

(event EventOnDamageServer
  (params (Damage :type Float%20%28single-precision%29) (DamageTags :type Gameplay%20Tag%20Container%20Structure%20%28by%20ref%29) (Momentum :type Vector :default "0, 0, 0") (HitInfo :type Hit%20Result%20Structure%20%28by%20ref%29) (InstigatedBy :type Controller%20Object%20Reference) (DamageCauser :type Actor%20Object%20Reference) (EffectContext :type Gameplay%20Effect%20Context%20Handle%20Structure))
  (|CallParentFunction Damage DamageTags Momentum HitInfo InstigatedBy DamageCauser EffectContext)
  (Components|DestroyComponent (Variables|Default|GetProximityTrigger)))

(event Custom|EventOnThroneDamaged
  (params (DamagedActor :type Actor%20Object%20Reference) (Damage :type Float%20%28single-precision%29) (InstigatedBy :type Controller%20Object%20Reference) (DamageCauser :type Actor%20Object%20Reference) (HitLocation :type Vector) (FHitComponent :type Primitive%20Component%20Object%20Reference) (BoneName :type Name) (Momentum :type Vector))
  (bind _asbuilding_actor (Utilities|Casting|CastToBuildingActor DamagedActor)
    (:then
      (CallFunction|TryTriggerAttackIntro InstigatedBy (<= (Health|GetHealth _asbuilding_actor :!class /Script/FortniteGame.BuildingActor) 0)))
    (:CastFailed
      (CallFunction|TryTriggerAttackIntro InstigatedBy))))

(event Custom|WakeUpAIFromIntro
  (params (FortPlayerPawnToSetHostileTo :type Fort%20Player%20Pawn%20Object%20Reference) (ByPassAnimation :type Boolean))
  ;[1 1 1] Send Event to Ability to stop Ability
  (Ability|SendGameplayEventtoActor self (Variables|Default|GetTag_EventWhenPawnInAttackProximity) (Utilities|Struct|MakeGameplayEventData (Variables|Default|GetTag_EventWhenPawnInAttackProximity) FortPlayerPawnToSetHostileTo 0 0 0 "()" (select ByPassAnimation (Variables|Default|GetInstigatorTagsForIntro) 0))))

(event Custom|TryTriggerAttackIntro
  (params (InstigatorController :type Controller%20Object%20Reference) (ShouldByPassIntro :type Boolean))
  ;[1 1 1] Look who attacked or destroyed the throne and then if we are still seated on the throne, get up!
  (bind _asfort_player_controller (Utilities|Casting|CastToFortPlayerController InstigatorController)
    (:then
      (Variables|Default|SetThroneAttacker (Game|Player|GetPlayerPawn _asfort_player_controller)))
    (:CastFailed
      (bind _asfort_athena_aibot_controller (Utilities|Casting|CastToFortAthenaAIBotController InstigatorController))
      (bind _asfort_player_pawn (Utilities|Casting|CastToFortPlayerPawn (Pawn|GetControlledPawn _asfort_athena_aibot_controller))
        (:then
          (Variables|Default|SetThroneAttacker _asfort_player_pawn))
        (:CastFailed
          _))
      (CallFunction|WakeUpAIFromIntro (Variables|Default|GetThroneAttacker) ShouldByPassIntro)
      (Components|DestroyComponent (Variables|Default|GetProximityTrigger)))))

(event Custom|EventOnThroneDied
  (params (DamagedActor :type Actor%20Object%20Reference) (Damage :type Float%20%28single-precision%29) (InstigatedBy :type Controller%20Object%20Reference) (DamageCauser :type Actor%20Object%20Reference) (HitLocation :type Vector) (FHitComponent :type Primitive%20Component%20Object%20Reference) (BoneName :type Name) (Momentum :type Vector))
  (CallFunction|TryTriggerAttackIntro InstigatedBy true))

(event Custom|BossDBNO
  (bind _timedbno (Variables|Default|GetTimeDBNO))
  ;[0.15 0.5 0.15 0.5] Team Convert Behavior
  (CallFunction|UnbindonConverted)
  (CallFunction|DisableCarry)
  (Networking|SwitchHasAuthority
    (:Authority
      (CallFunction|DropLoot)
      ;;[0 0.4903 1] Set State before Conversion
      (CallFunction|ActivateInvincibleVisuals _timedbno)
      (Fortnite|SetCanbeDamaged self)
      (CallFunction|ForceHideHealthBar 0.0)
      ;[1 1 1] stop movement here
      (Utilities|Time|SetTimerbyEvent (EventDispatchers|CreateEvent :!event Custom|JoinPlayerAnim) _timedbno))))

(event Custom|JoinPlayerAnim
  ;[0.15 0.5 0.15 0.5] Team Convert Behavior
  ;;[0.9307 0.3333 1 0.5] Play stand up montage before conversion
  (if (Movement|IsMovingonGround (Variables|Character|GetCharacterMovement))
    (bind _returnvalue (Animation|PlayAnimMontage (Variables|Default|GetConvertMontage)))
    (Utilities|Time|SetTimerbyEvent (EventDispatchers|CreateEvent :!event Custom|ConvertBossPostAnim) _returnvalue)
    (else
      (CallFunction|ConvertBossPostAnim))))

(event Custom|ConvertBossPostAnim
  ;[0.15 0.5 0.15 0.5] Team Convert Behavior
  (Convert|SetCanbeConverted (Variables|FortPlayerPawn|GetConvertComponent) true)
  ;;[0.8756 1 0 0.5] Check we allow auto conversion
  (if (and (ScalableFloat|GetValueasBool (Variables|Default|GetAutoConvertOnDBNO)) (Variables|Default|GetValidLastDamagingPlayer))
    ;;[0.9307 0.3333 1 0.5] Make sure montage is no longer playing before conversion
    (Animation|StopAnimMontage (Variables|Default|GetConvertMontage))
    ;;[0.9307 0.3333 1 0.5] Convert to Companion
    (Convert|TriggerConversion (Variables|Default|GetLastDamagingPawn) self)))

(event Custom|OnConverted
  (params (InstigatorPawn :type Fort%20Pawn%20Object%20Reference) (ConvertedPawn :type Fort%20Pawn%20Object%20Reference))
  ;[0.15 0.5 0.15 0.5] Team Convert Behavior
  ;;[1 1 1] re-enable movement here
  (Animation|StopAnimMontage (Variables|Default|GetConvertMontage))
  (Variables|Default|SetbIsConverted true))

(event Custom|On%20Converted%20VO%20Finished
  ;[0.15 0.5 0.15 0.5] Team Convert Behavior
  (Components|Activation|Deactivate (Variables|Default|GetVO%20Amb%20Whispers%20Component)))

(fn AssignLastDamagingPawn
  (params (DamagingPawn :type Fort%20Player%20Pawn%20Object%20Reference))
  (if (Pawn|IsPlayerControlled DamagingPawn)
    (Variables|Default|SetLastDamagingPawn DamagingPawn)
    (Variables|Default|SetValidLastDamagingPlayer true)
    (else
      (Variables|Default|SetValidLastDamagingPlayer false))))

(fn ClientOnAlertLevelChanged
  (params (OldAlertLevel :type EAlertLevel%20Enum :default "Unaware") (NewAlertLevel :type EAlertLevel%20Enum :default "Unaware"))
  (bind _vo_amb_whispers_component (Variables|Default|GetVO%20Amb%20Whispers%20Component))
  (|CallParentFunction OldAlertLevel NewAlertLevel)
  (if (Components|Activation|IsActive (Variables|Default|GetVO%20Amb%20Whispers%20Component))
    (switch Utilities|FlowControl|Switch|SwitchonEAlertLevel NewAlertLevel
      (:Unaware
        _)
      (:Suspicious
        (Audio|Components|Audio|FadeOut _vo_amb_whispers_component 0.5 0.0 "Linear" :!class /Script/Engine.AudioComponent))
      (:LostTarget
        _)
      (:Threatened
        _))
    (Audio|Components|Audio|FadeIn _vo_amb_whispers_component 0.5 1.0 0.0 "Linear" :!class /Script/Engine.AudioComponent)))

(fn DisableCarry
  (bind _asfort_ability_system_component (Utilities|Casting|CastToFortAbilitySystemComponent (Ability|GetAbilitySystemComponent self)))
  (Tags|BlockAbilitiesWithTags _asfort_ability_system_component (Variables|Default|GetBeCarriedTag)))

(fn DropLoot
  ;[0 1 0.9712] Drop Loot for Downing Boss
  (bind _outloottodrop (Loot|PickLootDrops (Utilities|Select :Index (Variables|Default|GetPlaylistTeamSize) :Option%200 "None" :Option%201 (Variables|Default|GetLootTierName) :Option%202 "LTG_AxisThrow_Duo" :Option%203 "LTG_AxisThrow_Trio" :Option%204 "LTG_AxisThrow_Squad") 0 -1 "(GameplayTags=)"))
  (Class|AthenaFunctionLibrary|SpawnLootfromArray _outloottodrop (Transformation|GetActorLocation self) "0, 0, 0" 0 true true false 0 "AI" "NPCElimination"))

(fn GetClosestThrone
  (bind _self self)
  (bind _returnvalue (Transformation|GetActorLocation _self))
  (LocalVariables|SetDistanceToThrone 99999999.0)
  (bind _outactors (Actor|GetAllActorsOfClass "/FNM_AI_AxisThrow/Environments/Throne/Blueprints/BSG/BSG_AxisThrow_Throne_A.BSG_AxisThrow_Throne_A_C"))
  (if (> (Utilities|Array|Length _outactors) 0)
    (for _array_element _outactors
      (if (< (Math|Vector|Distance(Vector) (Transformation|GetActorLocation _array_element) _returnvalue) (LocalVariables|GetDistanceToThrone))
        (LocalVariables|SetDistanceToThrone (Math|Vector|Distance(Vector) (Transformation|GetActorLocation _array_element) _returnvalue))
        (LocalVariables|SetSelectedThrone _array_element)))
    (return (Utilities|IsValid (LocalVariables|GetSelectedThrone) :!class /Script/Engine.KismetSystemLibrary) (LocalVariables|GetSelectedThrone))
    (else
      (return false 0))))

(fn HandlePawnConverted
  (params (InstigatorPawn :type Fort%20Pawn%20Object%20Reference) (ConvertedPawn :type Fort%20Pawn%20Object%20Reference))
  (|CallParentFunction InstigatorPawn ConvertedPawn)
  (Variables|Default|SetLeader%20Team (Team|GetActorTeam InstigatorPawn))
  (Utilities|FlowControl|Sequence
    (:then_0
      (CallFunction|UpdateConvertIndicator))
    (:then_1
      (Networking|SwitchHasAuthority
        (:Authority
          (FortPawn|SetHealth (ScalableFloat|GetValueatLevel (Variables|Default|GetHealthToApplyOnConverted)))
          (Fortnite|SetCanbeDamaged self true))))))

(fn OnRep_bIsConverted
  (Networking|CanExecuteCosmeticEvents
    (:True
      (Variables|Default|GetbIsConverted
        (:then
          (bind _vo_amb_whispers_component (Variables|Default|GetVO%20Amb%20Whispers%20Component)
            (:then
              (Audio|Components|Audio|FadeOut _vo_amb_whispers_component 0.5 0.0 "Linear" :!class /Script/Engine.AudioComponent)
              (EventDispatchers|BindEventtoOnAudioFinished (Variables|Default|GetVO%20Amb%20Whispers%20Component) (EventDispatchers|CreateEvent :!event On%20Converted%20VO%20Finished)))
            (:else)))
        (:else)))))

(fn UnbindOnConverted)

(fn UpdateConvertIndicator
  (bind _convertcomponent (Variables|FortPlayerPawn|GetConvertComponent))
  (Networking|SwitchHasAuthority
    (:Remote
      (if (Math|Byte|Equal(Byte) (Team|GetActorTeam (Fortnite|GetFirstLocalFortPlayerController true)) (Variables|Default|GetLeader%20Team))
        (Convert|ShowIndicatorMarker _convertcomponent)
        (else
          (Convert|HideIndicatorMarker _convertcomponent))))))

(fn ConstructionScript
  (|CallParentFunction))
